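# Translucent deferred decal material for a DragonCart orb attack scorch mark, with animated glowing embers.
# A sphere-masked cell-noise bubble pattern, broken up by two counter-panning noise layers and added to inverted Valet damage AO, drives edge-to-center emissive colors and a smoothstepped hot core.
# Opacity takes the max of the AO-modulated decal mask and the glow, scaled by decal color alpha.
# Damage and bubble normals blend by a contrast-adjusted mask, then pass through FlattenNormal with negative flatness.
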Material /DragonCartNarrative/Assets/VFX/Orb_Attack/MD_DragonCart_Scorch_Decal1
graph {
 properties {
  bAutomaticallySetUsageInEditor=false
  blendMode=BLEND_Translucent
  materialDomain=MD_DeferredDecal
  pixelDepthOffsetMode=PDOM_Legacy
  shadingModel=MSM_Unlit
  usedShadingModels=Unlit
 }
 nodes [
  Constant()
  n1 = DecalColor()
  n2 = VectorParameter() {group=Emissive, parameterName=EmissiveEdge, defaultValue=<rgba 1 0.02 0 1>}
  n3 = VectorParameter() {group=Emissive, parameterName=EmissiveCenter, defaultValue=<rgba 1.5 0.2768 0 1>}
  n4 = SphereMask(A=TextureCoordinate(), B=Constant() {r=0.5}, Radius=Constant() {r=0.6}) {hardnessPercent=0}
  n5 = ScalarParameter() {defaultValue=1.25, parameterName=NoiseTile}
  n6 = TextureSample(UVs=Multiply(A=TextureCoordinate() {uTiling=1.5, vTiling=1.5}, B=NamedRerouteDeclaration(None=@n5) {name=NoiseTile, nodeColor=<rgba 0.0547 1 0 1>})) {samplerType=SAMPLERTYPE_LinearGrayscale, texture=/Game/Vehicles/Valet/ValetDamage_AO}
  n7 = OneMinus(None=@n6.G)
  n8 = Multiply(A=@n7, B=@n7)
  n9 = ScalarParameter() {defaultValue=0.5, parameterName=NoiseAdd}
  n10 = NamedRerouteUsage()
  n11 = Multiply(A=TextureCoordinate(), B=@n10)
  n12 = Multiply(
   A=@n4
   B=TextureSample(UVs=NamedRerouteDeclaration(None=@n11) {name=BubbleUVs, nodeColor=<rgba 1 0.7031 0 1>}) {samplerType=SAMPLERTYPE_LinearGrayscale, texture=/Game/Effects/Fort_Effects/Textures/Tile/T_CellNoise_Tile_M}.G
  )
  n13 = TextureCoordinate() {uTiling=0.8, vTiling=0.8}
  n14 = Multiply(A=Time(), B=ScalarParameter() {defaultValue=0.5, parameterName=PanSpeed})
  n15 = TextureSample(UVs=Panner(Coordinate=Multiply(A=@n13, B=@n10), Time=@n14) {speedX=0.05, speedY=-0.02}) {texture=/Game/Global/Textures/Default/Noise/T_CellNoise_01}
  n16 = TextureCoordinate() {uTiling=1.5, vTiling=1.5}
  n17 = TextureSample(UVs=Panner(Coordinate=Multiply(A=@n10, B=@n16), Time=@n14) {speedX=-0.03, speedY=0.04}) {texture=&T_CellNoise_01}
  n18 = Multiply(A=@n12, B=Max(A=Multiply(A=@n12) {constB=0.6}, B=LinearInterpolate(A=@n15.G, B=@n17.G)))
  n19 = ScalarParameter() {defaultValue=1.6, parameterName=NoisePower}
  n20 = Add(
   A=Multiply(A=@n4, B=Multiply(A=@n8, B=@n9))
   B=Multiply(A=Power(Base=@n18, Exp=@n19) {constExponent=1.5}, B=ScalarParameter() {defaultValue=3, parameterName=NoiseMult}) {constB=4}
  )
  n21 = Power(Base=@n20, Exp=ScalarParameter() {defaultValue=1.5, parameterName=HotPower}) {constExponent=1.5}
  n22 = ScalarParameter() {defaultValue=1.5, parameterName=CenterHotGlowMultiply}
  n23 = Constant()
  n24 = ScalarParameter() {defaultValue=0.3, parameterName=SmoothStepSize_HotYellow}
  n25 = Power(Base=@n20, Exp=ScalarParameter() {defaultValue=1.5, parameterName=CenterHotGlowPower}) {constExponent=3}
  n26 = Multiply(
   A=LinearInterpolate(
    A=LinearInterpolate(A=@n2.RGB, B=@n3.RGB, Alpha=@n21)
    B=Multiply(A=@n3.RGB, B=@n22) {constB=3}
    Alpha=SmoothStep(Min=@n23, Max=@n24, Value=@n25)
   )
   B=@n21
  )
  MaterialFunctionCall(Emissive=Multiply(A=@n1.RGB, B=@n26)) {materialFunction=&MF_FX_EmissiveAdjustments}
  n27 = Multiply(A=TextureCoordinate(), B=NamedRerouteUsage())
  n28 = NamedRerouteUsage()
  n29 = NamedRerouteUsage()
  n30 = Constant() {r=0.5}
  MaterialFunctionCall(
   Normal=LinearInterpolate(
    A=TextureSample(UVs=@n27) {samplerType=SAMPLERTYPE_Normal, texture=/Game/Vehicles/Valet/ValetDamage_N}.RGB
    B=TextureSample(UVs=@n28) {samplerType=SAMPLERTYPE_Normal, texture=/Game/Effects/Fort_Effects/Textures/Tile/T_CellNoise_Tile_N}.RGB
    Alpha=MaterialFunctionCall(In=@n29, Contrast=@n30) {materialFunction=&CheapContrast}.Result
   )
   Flatness=ScalarParameter() {defaultValue=-0.5, parameterName=NormalFlatness}
  ) {materialFunction=&FlattenNormal}
  n31 = TextureSampleParameter2D(UVs=MaterialFunctionCall(Texture Scale=ScalarParameter() {defaultValue=1.1, parameterName=MaskTextureScale}) {materialFunction=&ScaleUVsByCenter}.UVs) {parameterName=DecalMask_Texture, samplerType=SAMPLERTYPE_LinearColor, texture=/Game/Effects/Fort_Effects/Textures/Decal/T_Decal_Default_A}
  n32 = Multiply(A=LinearInterpolate(Alpha=@n6.G) {constA=0.4}, B=@n31.G)
  n33 = Multiply(A=@n26) {constB=0.4}
  Multiply(A=@n1.A, B=Multiply(A=Max(A=@n32, B=@n33), B=ScalarParameter() {defaultValue=3, parameterName=OpacityMult}))
  NamedRerouteDeclaration(None=@n12) {name=BubbleMask, nodeColor=<rgba 0.3828 0 1 1>}
 ]
}
