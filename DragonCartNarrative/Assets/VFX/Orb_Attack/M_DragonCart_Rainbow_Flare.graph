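# Additive, unlit, two-sided Niagara material for a rainbow flare in the DragonCart orb attack VFX.
# Colour comes from a pannable rainbow texture, normalized, saturation-adjusted, tinted by parameter and particle colour, then fed through emissive adjustments.
# The flare mask is procedural: sine-warped UV scaling, driven by a dynamic dissolve offset, shapes two per-axis sine bands merged with Max.
# That mask is refined, then multiplied by edge fades and particle alpha.
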Material /DragonCartNarrative/Assets/VFX/Orb_Attack/M_DragonCart_Rainbow_Flare
graph {
 properties {
  bAutomaticallySetUsageInEditor=false
  bUsedWithNiagaraMeshParticles=true
  bUsedWithNiagaraRibbons=true
  bUsedWithNiagaraSprites=true
  blendMode=BLEND_Additive
  shadingModel=MSM_Unlit
  twoSided=true
  usedShadingModels=Unlit
 }
 nodes [
  Constant() {r=1}
  n1 = TextureObjectParameter() {group=Rainbow, parameterName=Rainbow_Texture, sortPriority=0, texture=&T_Tile_RainbowColor}
  n2 = VectorParameter() {group=Rainbow, parameterName=Rainbow_UVControl, channelNames {a=VPan, b=UPan, g=VScale, r=UScale}, defaultValue=<rgba 1 1 0.1 0>}
  n3 = Desaturation(
   None=Normalize(VectorInput=MaterialFunctionCall(Texture=@n1, UVControl_RGB=@n2, UVControl_A=@n2) {materialFunction=&FXCommon_TextureSample}.RGB)
   Fraction=ScalarParameter() {defaultValue=-0.5, parameterName=DES}
  )
  n4 = VectorParameter() {parameterName=C, defaultValue=<rgba 1.7913 4.4753 10 1>}
  n5 = ParticleColor()
  MaterialFunctionCall(Emissive=Multiply(A=Multiply(A=@n3, B=@n4.RGB), B=@n5.RGB)) {materialFunction=&MF_FX_EmissiveAdjustments}
  n6 = MaterialFunctionCall(Float2=TextureCoordinate()) {materialFunction=&BreakOutFloat2Components}
  n7 = VectorParameter() {
   group=EdgeFade
   parameterName=EdgeFade_EdgeMask_Control
   channelNames {a="", b=Power, g=Multiply, r=Offset}
   defaultValue=<rgba -0.032 3.5386 3.3383 1>
  }
  n8 = MaterialFunctionCall!(UorV=@n6.R, EdgeControl=@n7) {materialFunction=/Game/Effects/Niagara/MasterMaterials/Functions/Common/FXCommon_EdgeMask_Simple}
  n9 = MaterialFunctionCall(UorV=@n6.G, EdgeControl=@n7) {materialFunction=&FXCommon_EdgeMask_Simple}
  n10 = TextureCoordinate()
  n11 = ScalarParameter() {defaultValue=0.0784, parameterName=SCACA}
  n12 = Sine(None=Multiply(A=TextureCoordinate()) {constB=0.5})
  n13 = MaterialFunctionCall(Float2=Power(Base=OneMinus(None=@n12), Exp=ScalarParameter() {defaultValue=0.7019, parameterName=Test})) {materialFunction=&BreakOutFloat2Components}
  n14 = Add(
   A=ScalarParameter() {defaultValue=0.0628, parameterName=Dissolve}
   B=DynamicParameter() {defaultValue=<rgba 0 1 1 1>, paramNames [OffsetDissolve, Param2, Param3, Param4]}.OffsetDissolve
  )
  n15 = Clamp(None=MaterialFunctionCall(Input=@n13.G, Target Low=@n14) {materialFunction=&RemapValueRange}.Result) {clampMode=CMODE_ClampMin}
  n16 = Clamp(None=MaterialFunctionCall(Input=@n13.R, Target Low=@n14) {materialFunction=&RemapValueRange}.Result) {clampMode=CMODE_ClampMin}
  n17 = MaterialFunctionCall(Float2=MaterialFunctionCall(
    UVs=MaterialFunctionCall(UVs=@n10, Texture Scale=@n11) {materialFunction=&ScaleUVsByCenter}.UVs
    Texture Scale=AppendVector(A=@n15, B=@n16)
   ) {materialFunction=&ScaleUVsByCenter}.UVs) {materialFunction=&BreakOutFloat2Components}
  n18 = Constant2Vector() {g=1, r=1}
  n19 = ScalarParameter() {defaultValue=2, parameterName=Period}
  n20 = StaticBool() {value=true}
  n21 = MaterialFunctionCall(Value=@n17.R, Period=@n19, -1 to 1=@n20) {materialFunction=&LinearSine}
  n22 = MaterialFunctionCall(Value=@n17.G, Period=@n19, -1 to 1=@n20) {materialFunction=&LinearSine}
  n23 = Constant2Vector() {g=1, r=1}
  n24 = Max(
   A=Multiply(
    A=MaterialFunctionCall!(TexCoord=@n17.R, ClampChannel=@n18) {materialFunction=/Game/Effects/Fort_Effects/MaterialFunctions/MF_FX_TilingMethod}.Result
    B=Saturate(None=@n21.Linear Sine)
   )
   B=Multiply(
    A=Saturate(None=@n22.Linear Sine)
    B=MaterialFunctionCall(TexCoord=@n17.G, ClampChannel=@n23) {materialFunction=&MF_FX_TilingMethod}.Result
   )
  )
  n25 = VectorParameter() {parameterName=O_Refine, defaultValue=<rgba 5.0277 8.9748 0.388 0.5>}
  Multiply(
   A=Multiply(A=@n5.A, B=Multiply(A=@n8.Mask, B=@n9.Mask))
   B=Saturate(None=MaterialFunctionCall(Texture=@n24, RGB=@n25.RGB, A=@n25.RGB) {materialFunction=&MF_FX_Refine}.Result)
  )
 ]
}
