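# Translucent unlit material for a soft, textured shockwave in the Orb Attack VFX, built for Niagara mesh particles.
# Two Fresnel terms, one inverted, combine with a vertical gradient falloff and particle alpha to shape opacity, while particle color feeds emissive adjustments.
# Tiled noise erodes the shape, with dynamic parameters driving erosion, depth fade and a masked refraction index.
# A shading path switch skips the final depth fade on mobile.
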Material /DragonCartNarrative/Assets/VFX/Orb_Attack/M_DragonCart_Soft_Shockwave_Textured
graph {
 properties {
  bUsedWithNanite=true
  bUsedWithNiagaraMeshParticles=true
  blendMode=BLEND_Translucent
  pixelDepthOffsetMode=PDOM_Legacy
  refractionMethod=RM_IndexOfRefraction
  shadingModel=MSM_Unlit
  usedShadingModels=Unlit
 }
 nodes [
  Constant() {r=1}
  n1 = MaterialFunctionCall(
   In=OneMinus(None=MaterialFunctionCall() {materialFunction=&LinearGradient}.VGradient)
   Contrast=ScalarParameter() {defaultValue=0.05, parameterName=BottomFalloff}
  ) {materialFunction=&CheapContrast}
  n2 = Fresnel(ExponentIn=ScalarParameter() {defaultValue=1, parameterName=FresnelB_Exp}) {exponent=1}
  n3 = MaterialFunctionCall(
   In=Power(Base=OneMinus(None=@n2), Exp=ScalarParameter() {defaultValue=1, parameterName=FresnelB_Power})
   Contrast=ScalarParameter() {defaultValue=1, parameterName=FresnelB_Contrast}
  ) {materialFunction=&CheapContrast}
  LinearInterpolate(
   B=DynamicParameter() {defaultValue=<rgba 1 1 1 100>, paramNames [RefractionIndex, Erode, ErodeHardness, DepthFadeAmount]}.RefractionIndex
   Alpha=Multiply(A=@n1.Result, B=@n3.Result)
  ) {constA=1}
  n4 = ParticleColor()
  MaterialFunctionCall(Emissive=@n4.RGB) {materialFunction=&MF_FX_EmissiveAdjustments}
  n5 = ScalarParameter() {defaultValue=1, parameterName=FresnelA_Exp}
  n6 = Multiply(
   A=MaterialFunctionCall(In=Fresnel(ExponentIn=@n5) {exponent=1}, Contrast=ScalarParameter() {defaultValue=0.1, parameterName=FresnelA_Contrast}) {materialFunction=&CheapContrast}.Result
   B=@n3.Result
  )
  n7 = ScalarParameter() {defaultValue=1, parameterName=FresnelBrightness}
  n8 = DepthFade(
   Opacity=Clamp(None=Multiply(A=@n6, B=@n7))
   FadeDistance=DynamicParameter() {defaultValue=<rgba 1 1 1 100>, paramNames [RefractionIndex, Erode, ErodeHardness, DepthFadeAmount]}.DepthFadeAmount
  )
  n9 = DynamicParameter() {defaultValue=<rgba 1 1 1 100>, paramNames [RefractionIndex, Erode, ErodeHardness, DepthFadeAmount]}
  n10 = Multiply(
   A=TextureSample(UVs=TextureCoordinate() {uTiling=4, vTiling=4}) {texture=&T_NoisePack_01}.G
   B=@n9.ErodeHardness
  ) {constB=2}
  n11 = Multiply(A=@n9.ErodeHardness, B=@n9.Erode)
  n12 = Multiply(A=Multiply(A=@n4.A, B=Multiply(A=@n1.Result, B=@n8)), B=Clamp(None=Add(A=@n10, B=@n11)))
  ShadingPathSwitch(Default=DepthFade(Opacity=@n12, FadeDistance=ScalarParameter() {defaultValue=500, parameterName=FadeDistance}), Mobile=@n12) {desc=Migrated from FeatureLevelSwitch}
  Sine(None=Time())
 ]
}
