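# Unlit translucent material for glowing energy tendrils on Niagara mesh particles, filed under the Orb Attack VFX.
# Panning magnetic-wave and cloud noise, offset by vertex color, form a tendril mask alongside UV edge and end fades, particle alpha, and camera, angle and depth fades.
# Time-of-day colors blend purple to hot orange along V, tuned by dynamic parameters.
# SplineThicken offsets vertices using vertex-color-driven base and tip widths, halved on low quality.
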Material /DragonCartNarrative/Assets/VFX/Orb_Attack/M_SmartTuna_SolarPanelTendrils
graph {
 properties {
  bAutomaticallySetUsageInEditor=false
  bCastRayTracedShadows=false
  bUsedWithNiagaraMeshParticles=true
  bUsedWithStaticMesh=false
  blendMode=BLEND_Translucent
  pixelDepthOffsetMode=PDOM_Legacy
  shadingModel=MSM_Unlit
  twoSided=true
  usedShadingModels=Unlit
 }
 nodes [
  n1 = ParticleColor()
  MaterialFunctionCall(Emissive=Multiply(A=NamedRerouteUsage(), B=@n1.RGB)) {materialFunction=&MF_FX_EmissiveAdjustments}
  n2 = ComponentMask(None=TextureCoordinate()) {r=true}
  n3 = Multiply(A=Add(A=Cosine(None=@n2))) {constB=0.5}
  n4 = NamedRerouteUsage()
  n5 = Multiply(A=@n1.A, B=Multiply(A=OneMinus(None=@n3), B=Power(Base=@n4) {constExponent=0.8}))
  n6 = TextureCoordinate()
  n7 = Multiply(A=ComponentMask(None=@n6) {g=true}) {constB=3}
  n8 = Multiply(A=ComponentMask(None=OneMinus(None=@n6)) {g=true}) {constB=9}
  n9 = Saturate(None=Multiply(A=Saturate(None=@n7), B=Saturate(None=@n8)))
  NamedRerouteDeclaration(None=Multiply(A=Multiply(A=@n5, B=@n9), B=NamedRerouteUsage())) {name=Opacity, nodeColor=<rgba 1 0 0.7266 1>}
  n10 = Constant4Vector() {constant=<rgba 2.5201 0 2 1>}
  n11 = Constant4Vector() {constant=<rgba 0.586 0.0403 3 1>}
  n12 = Constant4Vector() {constant=<rgba 2.5201 0 2 1>}
  n13 = Constant4Vector() {constant=<rgba 2.5201 0 2 1>}
  n14 = Constant4Vector() {constant=<rgba 50 10.8415 0 1>}
  n15 = Constant4Vector() {constant=<rgba 60 11.8815 0.1215 1>}
  n16 = Constant4Vector() {constant=<rgba 50 10.8415 0 1>}
  n17 = Constant4Vector() {constant=<rgba 50 10.8415 0 1>}
  n18 = Power(
   Base=ComponentMask(None=TextureCoordinate()) {g=true}
   Exp=DynamicParameter() {defaultValue=<rgba 1 8 1 1>, paramNames [ColorPower,YellowReach,Param3,Param4]}.YellowReach
  ) {constExponent=8}
  n19 = Multiply(
   A=LinearInterpolate(
    A=MaterialFunctionCall(Color_Morning=@n10.RGB, Color_Day=@n11.RGB, Color_Evening=@n12.RGB, Color_Night=@n13.RGB) {materialFunction=&Color_Time_of_Day}.Result
    B=MaterialFunctionCall(Color_Morning=@n14.RGB, Color_Day=@n15.RGB, Color_Evening=@n16.RGB, Color_Night=@n17.RGB) {materialFunction=&Color_Time_of_Day}.Result
    Alpha=Saturate(None=@n18)
   )
   B=Add(A=NamedRerouteUsage()) {constB=0.2}
  )
  n20 = Multiply(A=@n19) {constB=0.1}
  n21 = ComponentMask(None=TextureCoordinate()) {r=true}
  n22 = Multiply(A=Add(A=Cosine(None=@n21))) {constB=0.5}
  NamedRerouteDeclaration(None=Power(
    Base=LinearInterpolate(A=@n20, B=@n19, Alpha=@n22)
    Exp=DynamicParameter() {defaultValue=<rgba 1 8 1 1>, paramNames [ColorPower,YellowReach,Param3,Param4]}.ColorPower
   )) {name=Color, nodeColor=<rgba 0 1 0.2266 1>}
  n23 = VertexColor()
  n24 = LinearInterpolate(
   A=ScalarParameter() {defaultValue=150, parameterName=WidthBase_Min}
   B=ScalarParameter() {defaultValue=300, parameterName=WidthBase_Max}
   Alpha=@n23.R
  )
  n25 = Constant() {r=1}
  n26 = QualitySwitch(Default=ShadingPathSwitch(Default=@n25), Low=Constant() {r=0.5}, High=@n25, Medium=@n25, Epic=@n25)
  n27 = LinearInterpolate(
   A=ScalarParameter() {defaultValue=25, parameterName=WidthTip_Min}
   B=ScalarParameter() {defaultValue=50, parameterName=WidthTip_Max}
   Alpha=@n23.G
  )
  NamedRerouteDeclaration(None=MaterialFunctionCall(WidthBase=Multiply(A=@n24, B=@n26), WidthTip=Multiply(A=@n27, B=@n26)) {materialFunction=&SplineThicken}.WorldPositionOffset) {name=WPO, nodeColor=<rgba 1 0 0.0937 1>}
  n28 = ScalarParameter() {parameterName=AngleFade}
  n29 = Saturate(None=Multiply(A=MaterialFunctionCall("FadeSensitivity (optional)"=@n28) {materialFunction=&FadeBasedOnViewAngle}.Result) {constB=2})
  n30 = Multiply(
   A=MaterialFunctionCall(Distance=ScalarParameter() {defaultValue=800, parameterName=NearFadeDIst}) {materialFunction=&Near_Camera_Fade}.Result
   B=DepthFade(FadeDistance=ScalarParameter() {defaultValue=100, parameterName=DepthFade})
  )
  NamedRerouteDeclaration(None=ShadingPathSwitch(Default=Multiply(A=@n29, B=@n30), Mobile=Constant() {r=0.7}) {desc=Migrated from FeatureLevelSwitch}) {name=Camera and Angle Fade, nodeColor=<rgba 1 0 0.9375 1>}
  n31 = AppendVector(A=Constant(), B=VertexColor().G)
  n32 = TextureCoordinate()
  n33 = VectorParameter() {parameterName=Tile1, defaultValue=<rgba 0.2 0.5 0 0>}
  n34 = AppendVector(A=@n33.R, B=@n33.R)
  n35 = Multiply(A=Add(A=Add(A=@n31, B=@n32), B=Multiply(B=@n34) {constA=-0.2, constB=0.5}), B=@n34)
  n36 = Multiply(A=@n35) {constB=1.5}
  n37 = VectorParameter() {parameterName=PanDirection1, defaultValue=<rgba 0.05 0.2 0 0>}
  n38 = AppendVector(A=@n37.R, B=@n37.R)
  n39 = Add(A=Multiply(A=Time(), B=ScalarParameter() {defaultValue=0.2, parameterName=PanSpeed}), B=VertexColor().R)
  n40 = VectorParameter() {parameterName=PanDirection1Mult, defaultValue=<rgba 1 0.5 0 1>}
  n41 = Multiply(A=@n38, B=AppendVector(A=@n40.R, B=@n40.R))
  n42 = Add(
   A=TextureSample(UVs=Panner(Coordinate=@n36, Time=@n39, Speed=@n38) {speedX=-0.2, speedY=-0.02}) {texture=/Game/Athena/Prototype/Blueprints/Sneezy/Textures/T_Sneezy_MagneticWaves}.RGB
   B=TextureSample(UVs=Panner(Coordinate=@n35, Time=@n39, Speed=@n41) {speedX=-0.2, speedY=0.02}) {texture=&T_Sneezy_MagneticWaves}.RGB
  )
  n43 = AppendVector(A=Constant(), B=VertexColor().G)
  n44 = TextureCoordinate()
  n45 = VectorParameter() {parameterName=Tile2, defaultValue=<rgba 1 0.4 0 0>}
  n46 = VectorParameter() {parameterName=PanDirection2, defaultValue=<rgba -0.1 0.15 0 0>}
  n47 = Panner(Coordinate=Multiply(A=Add(A=@n43, B=@n44), B=AppendVector(A=@n45.R, B=@n45.R)), Time=@n39, Speed=AppendVector(A=@n46.R, B=@n46.R)) {speedX=0.1, speedY=0.04}
  n48 = TextureSample(UVs=Add(A=@n47) {constB=0.01}) {samplerType=SAMPLERTYPE_LinearColor, texture=/Game/Global/Textures/Default/Noise/T_CloudNoise_01}
  n49 = TextureSample(UVs=@n47) {samplerType=SAMPLERTYPE_LinearColor, texture=&T_CloudNoise_01}
  NamedRerouteDeclaration(None=Add(A=Multiply(A=@n42) {constB=0.3}, B=Multiply(A=@n48.G, B=@n49.G))) {name=Tendrils, nodeColor=<rgba 0 1 0.0625 1>}
  NamedRerouteUsage()
  NamedRerouteUsage()
 ]
}
