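# Niagara system for the orb attack's death laser in DragonCartNarrative, a beam growing from the orb to an end point.
# Shared system values drive growth, flicker and fade-out, with the orb's WPO offsetting the origin.
# A mesh, sprite and light core joins an energy ribbon, distance-scaled lightning meshes and a muzzle flare.
# Radial-blur sprite and sparks scattered along the beam add impact; GiantGlow is disabled.
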NiagaraSystem /DragonCartNarrative/Assets/VFX/Orb_Attack/NS_DragonCartNarrative_Orb_DeathLaser
dependencies {
 usedDataInterfaces [
  &NiagaraDataInterfaceCurve
  &NiagaraDataInterfaceVectorCurve
  &NiagaraDataInterfaceMeshRendererInfo
  &NiagaraDataInterfaceCamera
 ]
 usedDynamicInputs [
  &self:ScratchDynamicInput_01
  &FNTimeOfDayScalar
  &NormalizeVector
  &Subtract_Position
  &self:ScratchDynamicInput
  &ReturnCameraPosition
  &Subtract_Float
 ]
 usedModules [
  &SystemState
  &self:SystemSpawnScript.SystemScriptSource.SystemScriptGraph.NiagaraNodeAssignment_8.SetVariables_#1
  &self:SystemSpawnScript.SystemScriptSource.SystemScriptGraph.NiagaraNodeAssignment_10.SetVariables_#2
  &self:SystemSpawnScript.SystemScriptSource.SystemScriptGraph.NiagaraNodeAssignment_1.SetVariables_#3
  &self:SystemSpawnScript.SystemScriptSource.SystemScriptGraph.NiagaraNodeAssignment_2.SetVariables_#4
  &InterpolateOverTime
  &self:SystemSpawnScript.SystemScriptSource.SystemScriptGraph.NiagaraNodeAssignment_5.SetVariables_#5
  &self:SystemSpawnScript.SystemScriptSource.SystemScriptGraph.NiagaraNodeAssignment_4.SetVariables_#6
  &self:SystemSpawnScript.SystemScriptSource.SystemScriptGraph.NiagaraNodeAssignment_0.SetVariables_#7
  &self:SystemSpawnScript.SystemScriptSource.SystemScriptGraph.NiagaraNodeAssignment_6.SetVariables_#8
  &self:SystemSpawnScript.SystemScriptSource.SystemScriptGraph.NiagaraNodeAssignment_7.SetVariables_#9
  &self:SystemSpawnScript.SystemScriptSource.SystemScriptGraph.NiagaraNodeAssignment_3.SetVariables_#10
  &self:SystemSpawnScript.SystemScriptSource.SystemScriptGraph.NiagaraNodeAssignment_11.SetVariables_#11
 ]
}
emitters {
 Beam_Core_and_MF {
  emitterUpdateScript [
   &EmitterState(
    Inactive Response=::NewEnumerator1
    Life Cycle Mode=::NewEnumerator1
    Loop Behavior=::NewEnumerator1
    Loop Duration Mode=::NewEnumerator1
    UseLoopDelay=false
   )
   &SpawnBurst_Instantaneous(Use Spawn Probability=false, UseLoopCountLimit=false)
  ]
  particleSpawnScript [
   &InitializeParticle(
    Color=$System.Color 01
    Mesh Scale=<9 20 20>
    Mesh Scale Mode=::NewEnumerator4
    Position Offset=<60 0 0>
    Sprite Size Mode=::NewEnumerator3
    Uniform Sprite Size=3000
    UsePositionOffset=false
    Write Material Random=true
   )
   &CameraOffset(Camera Offset Amount=500)
   &DynamicMaterialParameters(
    Index 0 Param 1=0.5
    Index 0 Param 2=2
    Index 2 Param 1=1
    Index 2 Param 2=3
    Index 3 Param 1=&FloatFromCurve(FloatCurve=DataInterface() {propertyValues {Curve {keys [<at 0.8 0.1>,<at 1 1>]}}}, Scale Curve=30)
    Index 3 Param 3=0.5
    Index 3 Param 4=&FloatFromCurve(FloatCurve=DataInterface() {propertyValues {Curve {keys [<at 0 1>,<at 1 0>]}}}, Scale Curve=1.2)
    Module.Index0Float0Write=true
    Module.Index0Float1Write=true
    Module.Index0Float2Write=false
    Module.Index0Float3Write=false
    Module.Index2Float0Write=true
    Module.Index2Float1Write=true
    Module.Index2Float2Write=true
    Module.Index2Float3Write=true
    Module.Index3Float0Write=true
    Module.Index3Float1Write=false
    Module.Index3Float2Write=true
    Module.Index3Float3Write=true
    Write Parameter Index 2=::NewEnumerator1
    Write Parameter Index 3=::NewEnumerator1
   )
   &InitialMeshOrientation(Mesh Orientation Mode=::NewEnumerator2, Orientation Vector=<0 0 0>, Use Rotation Vector=false)
  ]
  particleUpdateScript [
   &ParticleState()
   &ScaleColor(
    Scale Alpha=&FloatFromCurve(
     FloatCurve=DataInterface() {propertyValues {Curve {
        defaultValue=0
        keys [{interpMode=RCIM_Cubic, tangentMode=RCTM_User, time=0.7, value=1}, {interpMode=RCIM_Cubic, tangentMode=RCTM_User, time=1, value=0}]
       }}}
     CurveIndex=$System.TimeForDis
     Scale Curve=$User.BeamOn
    )
    ScaleA=true
    ScaleRGB=false
   )
   &ScaleMeshSize(Scale Factor=&VectorFromCurve(
     VectorCurve=DataInterface() {propertyValues {
       XCurve {keys [{arriveTangent=2.7828, interpMode=RCIM_Cubic, leaveTangent=2.7828, tangentMode=RCTM_User, time=0, value=0}, <at 1 1>]}
       YCurve {keys [{arriveTangent=5.2793, interpMode=RCIM_Cubic, leaveTangent=5.2793, tangentMode=RCTM_User, time=0, value=0}, <at 0.5 1>]}
       ZCurve {keys [{arriveTangent=5.2793, interpMode=RCIM_Cubic, leaveTangent=5.2793, tangentMode=RCTM_User, time=0, value=0}, <at 0.5 1>]}
      }}
     Scale Curve=&Multiply_VectorByFloat(Vector=&Lerp_Vector(A=<1.05 1 1>, B=<1 1.04 1.04>, Alpha=$System.Flicker), Float=$System.NearFade)
    ))
   &UpdateMeshOrientation(
    Facing Direction=<0 -0.25 0>
    Facing Interpolation=false
    Orientation Method=::NewEnumerator1
    Override Quat=false
    TransformFacing=false
    TransformUp=false
    Up Direction:Vector3f=<0 0 0>
   )
   &SpriteFacingAndAlignment(
    Alignment Coordinate Space=::World
    Facing Coordinate Space=::World
    Sprite Alignment=$System.ForwardDirection
    Sprite Facing=$Engine.Owner.SystemXAxis
    WriteAlignment=true
    WriteFacing=true
   )
   /SetVariables(Particles.Position=&AddVectorToPosition(
     Position=$System.StartPoint
     Vector=&TransformVector(Vector=$System.OrbWPO, Source Space=::World, Destination Space=::Local)
    )) {isSetParameters=true}
   /SetVariables(Particles.LightPosition:NiagaraPosition=&AddVectorToPosition(
     Position=&AddVectorToPosition(
      Position=$System.StartPoint
      Vector=&TransformVector(Vector=$System.OrbWPO, Source Space=::World, Destination Space=::Local)
     )
     Vector=&Multiply_VectorByFloat(Vector=$System.ForwardDirection, Float=1600)
    )) {isSetParameters=true}
   /SetVariables(
    Particles.BeamSpriteDyn:Vector4f=&MakeVector4(
     X=&FloatFromCurve(
      FloatCurve=DataInterface() {propertyValues {Curve {defaultValue=1, keys [{tangentMode=RCTM_User, time=0.6, value=0.2}, {tangentMode=RCTM_User, time=1, value=1}]}}}
      Scale Curve=5
     )
     Y=300
     Z=3
     W=1
    )
    Particles.BeamSpritePosition:NiagaraPosition=&LerpPosition(
     A=&AddVectorToPosition(
      Position=$System.StartPoint
      Vector=&TransformVector(Vector=$System.OrbWPO, Source Space=::World, Destination Space=::Local)
     )
     B=$System.EndPoint
     Alpha=&Multiply_Float(A=$System.BeamGrowth, B=0.5)
    )
    Particles.BeamSpriteSize:Vector2f=&MakeVector2D(
     X=&Multiply_Float(
      A=&Lerp_Float(A=300, B=600, Alpha=$System.Flicker)
      B=&FloatFromCurve(
       FloatCurve=DataInterface() {propertyValues {Curve {
          defaultValue=1
          keys [
           <at 0 0.5 RCIM_Cubic>
           {arriveTangent=-0.5551, interpMode=RCIM_Cubic, leaveTangent=-0.5551, tangentMode=RCTM_User, time=0.4, value=1}
           {arriveTangent=0.2288, interpMode=RCIM_Cubic, leaveTangent=0.2288, tangentMode=RCTM_User, time=1, value=0}
          ]
         }}}
       CurveIndex=$System.TimeForDis
       Scale Curve=$User.BeamOn
      )
     )
     Y=&Add_Float(
      A=1
      B=&Multiply_Float(A=&DistanceBetweenPositions(A=$System.StartPoint, B=$System.EndPoint), B=$System.BeamGrowth)
     )
    )
   ) {isSetParameters=true}
   &ScaleSpriteSize(
    Uniform Curve Scale=&Multiply_Float(
     A=&Lerp_Float(A=0.9, B=1, Alpha=$System.Flicker)
     B=&Multiply_Float(
      A=&FloatFromCurve(
       FloatCurve=DataInterface() {propertyValues {Curve {
          defaultValue=0.5
          keys [{interpMode=RCIM_Cubic, tangentMode=RCTM_User, time=0.8, value=0.7}, {interpMode=RCIM_Cubic, tangentMode=RCTM_User, time=1, value=0}]
         }}}
       CurveIndex=$System.TimeForDis
       Scale Curve=$User.BeamOn
      )
      B=$System.NearFade
     )
    )
    Uniform Curve Sprite Scale=DataInterface() {propertyValues {Curve {keys [
        {arriveTangent=33.5633, interpMode=RCIM_Cubic, leaveTangent=33.5633, tangentMode=RCTM_User, time=0, value=0}
        {arriveTangent=-8.0736, interpMode=RCIM_Cubic, leaveTangent=-8.0736, tangentMode=RCTM_User, time=0.2074, value=3}
        {tangentMode=RCTM_User, time=1, value=1}
       ]}}}
   )
   /SetVariables(
    Particles.StartSpriteColor:LinearColor=&MultiplyLinearColor(
     Linear Color=$Particles.Color
     Scale Factor=&MakeVector4FromVectorAndFloat(
      "Vector (RGB)"=&VectorFromFloat(Value=&Lerp_Float(A=12, B=16, Alpha=$System.Flicker))
      "Float (Alpha)"=1
     )
    )
    Particles.StartSpritePosition:NiagaraPosition=&AddVectorToPosition(Position=$Particles.Position, Vector=<0 0 -50>)
   ) {isSetParameters=true}
   /SetVariables(Particles.LightColor=&Multiply_LinearColorByFloat(Linear Color=$Particles.Color, Float=&Lerp_Float(A=12, B=16, Alpha=$System.Flicker))) {isSetParameters=true}
   /SetVariables(Particles.LightRadius=&Multiply_Float(
     A=10000
     B=&FloatFromCurve(
      FloatCurve=DataInterface() {propertyValues {Curve {keys [<at 0.5 1 RCIM_Cubic>,<at 1 0 RCIM_Cubic>]}}}
      CurveIndex=$System.TimeForDis
      Scale Curve=$User.BeamOn
     )
    )) {isSetParameters=true}
   /SetVariables(Particles.Wind Color:LinearColor=&Multiply_LinearColorByFloat(
     Linear Color=$Particles.Initial.Color
     Float=&FloatFromCurve(
      FloatCurve=DataInterface() {propertyValues {Curve {keys [
          {arriveTangent=-0.5579, interpMode=RCIM_Cubic, leaveTangent=-0.5579, tangentMode=RCTM_User, time=0, value=1}
          {arriveTangent=0.0179, tangentMode=RCTM_User, time=1, value=0}
         ]}}}
      Scale Curve=2
     )
    )) {isSetParameters=true}
   /SetVariables(
    Particles.FlatFlareColor:LinearColor=&MultiplyLinearColor(
     Linear Color=&Multiply_LinearColorByFloat(Linear Color=$User.Color 02, Float=0.8)
     Scale Factor=&MakeFloat4FromVectorAndFloat(
      "Vector (RGB)"=&Multiply_VectorByFloat(
       Vector=&VectorFromFloat(Value=&Waveform("Waveform [1]"=::NewEnumerator12, "[1] Frequency"=10, "[1] Amplitude Min/Max"=<0.8 1.1>, "[1] Scale"=0.9))
       Float=1
      )
      "Float (Alpha)"=&FloatFromCurve(
       FloatCurve=DataInterface() {propertyValues {Curve {keys [{interpMode=RCIM_Cubic, tangentMode=RCTM_User, time=0, value=1}, <at 1 0>]}}}
       CurveIndex=$System.TimeForDis
      )
     )
    )
    Particles.FlatFlareDyn:Vector4f=&MakeVector4(
     X=0
     Y=&Waveform("Waveform [1]"=::NewEnumerator12, "[1] Frequency"=20, "[1] Amplitude Min/Max"=<0 0.2>)
     Z=0
     W=0
    )
    Particles.FlatFlareSize:Vector2f=&Multiply_Vector2D_ByFloat(
     Float=1.5
     Vector2D=&Multiply_Vector2(
      A=&Multiply_Vector2D(A=<-2500 200>, B=<1 1>)
      B=&Vector2DFromFloat(Value=&Multiply_Float(A=5, B=1))
     )
    )
   ) {isSetParameters=true}
  ]
  rendererClasses [
   &NiagaraSpriteRendererProperties
   &NiagaraLightRendererProperties
   &NiagaraMeshRendererProperties
  ]
  renderers [
   {rendererClass=&NiagaraSpriteRendererProperties}
   {rendererClass=&NiagaraLightRendererProperties, rendererIndex=1}
   {rendererClass=&NiagaraSpriteRendererProperties, rendererIndex=2}
   {rendererClass=&NiagaraMeshRendererProperties, rendererIndex=3}
   {rendererClass=&NiagaraSpriteRendererProperties, rendererIndex=4}
  ]
 }
 Beam_Deactivate_Particles {
  emitterUpdateScript [
   &EmitterState(Life Cycle Mode=::NewEnumerator1, Loop Behavior=::NewEnumerator1, Loop Delay=0.2, Loop Duration=$User.Duration, UseLoopDelay=true)
   &SpawnPerFrame(Spawn:NiagaraBool=$System.DoOnce.Execute, Spawn Count=30, Use Spawn Probability=false) {enabled=false}
   &SpawnBurst_Instantaneous(Spawn Count=50, Use Spawn Probability=false, UseLoopCountLimit=false)
  ]
  particleSpawnScript [
   &InitializeParticle(
    Color=&Multiply_LinearColorByFloat(Linear Color=$User.Color 01, Float=3)
    Lifetime=&RandomRangeFloat(Minimum=0.3, Maximum=2)
    Position=&LerpPosition(A=$System.StartPoint, B=$System.EndPoint, Alpha=&ReturnNormalizedExecIndex())
    Position Mode=::NewEnumerator1
    Sprite Size=&RandomRangeVector2D(Minimum=<5 50>, Maximum=<7 100>)
    Sprite Size Mode=::NewEnumerator4
    UsePositionOffset=false
    Write Material Random=true
   )
   &ShapeLocation(
    Cylinder Height=1
    Cylinder Height Midpoint=0
    Cylinder Radius=1
    MaskSpawnGroup=false
    Override Seed=false
    Shape Primitive=::NewEnumerator1
    Surface Only=false
    UseOwnerScale=true
   )
   &AddVelocity(
    Rotation Coordinate Space=::World
    Velocity=$System.ForwardDirection
    Velocity Speed Scale=&RandomRangeFloat(Minimum=2000, Maximum=5000)
   )
  ]
  particleUpdateScript [
   &ParticleState()
   &ScaleSpriteSize(
    Uniform Curve Scale=0.2
    Uniform Curve Sprite Scale=DataInterface() {propertyValues {Curve {keys [{tangentMode=RCTM_User, time=0.5, value=1}, {tangentMode=RCTM_User, time=1, value=0}]}}}
   )
   &VortexForce(UseOriginOffset=false, Vortex Force Amount=5000) {enabled=false}
   &CurlNoiseForce(Bias Noise=false, Noise Frequency=100, Noise Strength=2500, Pan Noise=false)
   &Drag(Drag=&RandomRangeFloat(Minimum=1, Maximum=5), Use Linear Drag=true, Use Rotational Drag=true)
   &SolveForcesAndVelocity(Clamp Velocity=false, Limit Acceleration=false)
   &FN_Sprite_Attributes(
    Color=&MultiplyLinearColor(Linear Color=$Particles.Color, Scale Factor=<0.025 0.025 0.025 0.5>)
    Create Color Attribute=true
    Create SpriteSize Attribute=true
    SpriteSize=&Multiply_Vector2D(A=$Particles.SpriteSize, B=<10 2>)
   )
   &ScaleColor(
    Scale Alpha=&FloatFromCurve(FloatCurve=DataInterface() {propertyValues {Curve {
        defaultValue=1
        keys [
         {arriveTangent=9.0725, interpMode=RCIM_Cubic, leaveTangent=9.0725, tangentMode=RCTM_User, time=0, value=0}
         {arriveTangent=-0.3964, interpMode=RCIM_Cubic, leaveTangent=-0.3964, tangentMode=RCTM_User, time=0.1408, value=0.9847}
         {arriveTangent=-0.0405, interpMode=RCIM_Cubic, leaveTangent=-0.0405, tangentMode=RCTM_User, time=1, value=0}
        ]
       }}})
    ScaleA=true
    ScaleRGB=true
   )
  ]
  rendererClasses [&NiagaraSpriteRendererProperties]
  renderers [{rendererClass=&NiagaraSpriteRendererProperties}]
 }
 Beam_Lightning {
  emitterUpdateScript [
   &EmitterState(Life Cycle Mode=::NewEnumerator1, Loop Behavior=::NewEnumerator1, Loop Delay=0.1, Loop Duration Mode=::NewEnumerator1, UseLoopDelay=false)
   &SpawnBurst_Instantaneous(Spawn Count=3, Spawn Time=0.15, Use Spawn Probability=false, UseLoopCountLimit=true)
   &SpawnRate(
    SpawnRate=&MakeCustomFloatFromBool(Bool=&SetBoolByFloatComparison(A=$System.NearFade, B=0), True Float=8, False Float=0)
    Use Spawn Probability=false
   ) {enabled=false}
  ]
  particleSpawnScript [
   &PartitionParticles(Number Of Partitions=3)
   &InitializeParticle(
    Lifetime=&Multiply_Float(A=$User.DisappearTime, B=&RandomRangeFloat(Minimum=0.5, Maximum=1))
    Mesh Renderer Array Visibility Mode=::NewEnumerator1
    Mesh Renderer Info=DataInterface() {propertyValues {MeshRenderer=&self:Beam_Lightning_0.NiagaraMeshRendererProperties_0}}
    Mesh Renderer Mesh Index:NiagaraInt32=$Particles.PartitionParticles.Partition
    Mesh Scale=&MakeVector(
     X=1
     Y=&RandomRangeFloat(Minimum=4, Maximum=8)
     Z=&RandomRangeFloat(Minimum=4, Maximum=8)
    )
    Mesh Scale Mode=::NewEnumerator4
    UsePositionOffset=false
    Write Material Random=true
   )
   &InitialMeshOrientation(
    Mesh Orientation Mode=::NewEnumerator2
    Orientation Vector=<-1 0 0>
    Rotation=&RandomRangeVector(Minimum=<0 0 0>, Maximum=<1 0 0>)
    Use Rotation Vector=true
   )
  ]
  particleUpdateScript [
   &ParticleState()
   /SetVariables(Particles.Position=&AddVectorToPosition(
     Position=$System.EndPoint
     Vector=&Multiply_VectorByFloat(Vector=$System.ForwardDirection, Float=&Lerp_Float(A=0, B=500, Alpha=$Particles.MaterialRandom))
    )) {isSetParameters=true}
   /SetVariables(
    Particles.Flare:NiagaraFloat=&Lerp_Float(
     A=-0.4
     B=&FloatFromCurve(FloatCurve=DataInterface() {propertyValues {Curve {keys [
          {arriveTangent=-9.9823, interpMode=RCIM_Cubic, time=0.0073, value=1}
          {arriveTangentWeight=0.1021, interpMode=RCIM_Cubic, leaveTangentWeight=0.2548, tangentMode=RCTM_User, time=0.35, value=0}
          {arriveTangentWeight=1.9969, interpMode=RCIM_Cubic, leaveTangent=19.1881, leaveTangentWeight=0.804, tangentMode=RCTM_Break, time=0.5, value=0}
          {arriveTangent=-4.9513, interpMode=RCIM_Cubic, leaveTangent=-4.9513, tangentMode=RCTM_User, time=0.6, value=1}
          <at 0.9 0>
         ]}}})
     Alpha=&FloatFromCurve(FloatCurve=DataInterface() {propertyValues {Curve {keys [<at 0.2 1>,<at 0.5 0>]}}}, CurveIndex=$System.TimeForDis, Scale Curve=$User.BeamOn)
    )
    Particles.Flicker:NiagaraFloat=&Waveform("Waveform [1]"=::NewEnumerator0, "[1] Frequency"=25, "[1] Amplitude Min/Max"=<0.5 1>)
   ) {isSetParameters=true}
   &ScaleMeshSize(Scale Factor=&VectorFromCurve(
     VectorCurve=DataInterface() {propertyValues {
       XCurve {keys [
         {interpMode=RCIM_Cubic, leaveTangentWeight=0.1333, tangentMode=RCTM_User, time=0.6, value=1}
         {arriveTangent=-2.2924, tangentMode=RCTM_User, time=1, value=0.5}
        ]}
       YCurve {keys [
         {arriveTangent=3.3015, interpMode=RCIM_Cubic, leaveTangent=3.3015, tangentMode=RCTM_User, time=0, value=0.25}
         {arriveTangentWeight=0.1954, interpMode=RCIM_Cubic, leaveTangentWeight=0.1466, tangentMode=RCTM_User, time=0.4, value=1}
        ]}
       ZCurve {keys [
         {arriveTangent=3.3015, interpMode=RCIM_Cubic, leaveTangent=3.3015, tangentMode=RCTM_User, time=0, value=0.25}
         {arriveTangentWeight=0.1954, interpMode=RCIM_Cubic, leaveTangentWeight=0.1466, tangentMode=RCTM_User, time=0.4, value=1}
        ]}
      }}
     Scale Curve=&MakeVector(
      X=&Divide_Float(
       A=&DistanceBetweenPositions(
        A=&AddVectorToPosition(
         Position=$System.StartPoint
         Vector=&TransformVector(Vector=$System.OrbWPO, Source Space=::World, Destination Space=::Local)
        )
        B=$System.EndPoint
       )
       B=1000
      )
      Y=&Add_Float(A=1, B=$Particles.Flare)
      Z=&Add_Float(A=1, B=&Multiply_Float(A=$Particles.Flare, B=0.5))
     )
    ))
   &ScaleColor(
    Scale Alpha=&FloatFromCurve(FloatCurve=DataInterface() {propertyValues {Curve {keys [<at 0.4 1 RCIM_Cubic>,<at 1 0 RCIM_Cubic>]}}})
    Scale RGB=&VectorFromFloat(Value=&Multiply_Float(A=$Particles.Flicker, B=&Lerp_Float(A=1, B=1.5, Alpha=$Particles.Flare)))
    ScaleA=true
    ScaleRGB=true
   )
   &CameraOffset(Camera Offset Amount=&Multiply_Float(
     A=/Niagara/DynamicInputs/Math/CalculateParticleRadius(
      Radius Calculation Type:ENiagaraCollisionRadiusOptions=::NewEnumerator0
      Particle Radius:NiagaraFloat=1
      Mesh Dimensions:Vector3f=<1 1 1>
      Method for Calculating Particle Radius:ENiagaraCalculateRadiusOptions=::NewEnumerator2
     )
     B=0.5
    ))
   &UpdateMeshOrientation(
    Facing Direction=$System.ForwardDirection
    Facing Interpolation=false
    Orientation Method=::NewEnumerator1
    Override Quat=false
    TransformFacing=false
    TransformUp=false
   ) {module=UpdateMeshOrientation001}
   &UpdateMeshOrientation(Rotation Rate=0.25)
   &DynamicMaterialParameters(
    Index 0 Param 1=&FloatFromCurve(FloatCurve=DataInterface() {propertyValues {Curve {
        defaultValue=1
        keys [
         {arriveTangent=-0.1528, interpMode=RCIM_Cubic, leaveTangent=-0.1528, tangentMode=RCTM_User, time=0, value=1}
         {arriveTangent=-2.3611, tangentMode=RCTM_User, time=0.53, value=0.5}
         {arriveTangent=-1.9699, tangentMode=RCTM_User, time=1, value=0}
        ]
       }}})
    Index 0 Param 2=&Lerp_Float(A=0.2, B=0.4, Alpha=$Particles.Flare)
    Index 0 Param 3=&Lerp_Float(A=0.1, B=0.3, Alpha=$Particles.Flare)
    Index 0 Param 4=$Particles.MaterialRandom
    Index 1 Param 1=&FloatFromCurve(
     FloatCurve=DataInterface() {propertyValues {Curve {defaultValue=1, keys [{tangentMode=RCTM_User, time=0, value=0}, {tangentMode=RCTM_User, time=1, value=1}]}}}
     Scale Curve=10
    )
    Index 1 Param 2=7
    Index 1 Param 3=0.5
    Index 2 Param 1=1000
    Index 3 Param 2=1
    Module.Index0Float0Write=true
    Module.Index0Float1Write=true
    Module.Index0Float2Write=true
    Module.Index0Float3Write=true
    Module.Index1Float0Write=true
    Module.Index1Float1Write=true
    Module.Index1Float2Write=true
    Module.Index1Float3Write=false
    Module.Index2Float0Write=true
    Module.Index2Float1Write=false
    Module.Index2Float2Write=false
    Module.Index2Float3Write=false
    Module.Index3Float0Write=false
    Module.Index3Float1Write=false
    Module.Index3Float2Write=false
    Module.Index3Float3Write=false
    Write Parameter Index 1=::NewEnumerator1
    Write Parameter Index 2=::NewEnumerator1
    Write Parameter Index 3=::NewEnumerator1
   )
  ]
  rendererClasses [&NiagaraMeshRendererProperties]
  renderers [{rendererClass=&NiagaraMeshRendererProperties, rendererIndex=0}]
 }
 Excess_Energy_Beam {
  emitterUpdateScript [
   &EmitterState(Life Cycle Mode=::NewEnumerator1, Loop Behavior=::NewEnumerator1, Loop Duration Mode=::NewEnumerator1, UseLoopDelay=false)
   &SpawnBurst_Instantaneous(Spawn Count=20, Use Spawn Probability=false, UseLoopCountLimit=false)
  ]
  particleSpawnScript [&InitializeParticle(
    Color=$User.Color 01
    Lifetime=$System.Beam Duration
    Ribbon Width=&FloatFromCurve(
     FloatCurve=DataInterface() {propertyValues {Curve {keys [
         {arriveTangent=-11.4343, interpMode=RCIM_Cubic, time=0.1, value=0.7}
         {arriveTangent=0.6996, interpMode=RCIM_Cubic, leaveTangent=0.6996, time=0.3, value=1.5}
         {arriveTangent=-0.8333, interpMode=RCIM_Cubic, leaveTangent=-0.8333, time=0.8147, value=1.2}
         <at 0.9 1>
        ]}}}
     CurveIndex=&ReturnNormalizedExecIndex()
     Scale Curve=1000
    )
    Ribbon Width Mode=::NewEnumerator1
    UsePositionOffset=false
    Write Material Random=true
   )]
  particleUpdateScript [
   &ParticleState()
   /SetVariables(Particles.NormalizedDistance:NiagaraFloat=&NormalizeDistanceRange(
     Start Position:NiagaraPosition=$System.StartPoint
     End Position:NiagaraPosition=$Particles.Position
     Distance:NiagaraFloat=&DistanceBetweenPositions(A=$System.StartPoint, B=$System.EndPoint)
     Invert Normalized Range:NiagaraBool=false
    )) {isSetParameters=true}
   /SetVariables(Particles.Position=&Lerp_Vector(
     A=&AddVectorToPosition(
      Position=$System.StartPoint
      Vector=&TransformVector(Vector=$System.OrbWPO, Source Space=::World, Destination Space=::Local)
     )
     B=&AddVectorToPosition(Position=$System.EndPoint, Vector=&Multiply_VectorByFloat(Vector=$System.ForwardDirection, Float=100))
     Alpha=&FloatFromCurve(
      FloatCurve=DataInterface() {propertyValues {Curve {
         defaultValue=1
         keys [
          {arriveTangent=-0.3457, interpMode=RCIM_Cubic, leaveTangent=-0.3457, tangentMode=RCTM_User, time=0, value=0}
          {arriveTangent=-0.4774, interpMode=RCIM_Cubic, leaveTangent=-0.4774, tangentMode=RCTM_User, time=1, value=1}
         ]
        }}}
      CurveIndex=&ReturnNormalizedExecIndex()
      Scale Curve=$System.BeamGrowth
     )
    )) {isSetParameters=true}
   &ScaleRibbonWidth(Ribbon Width Scale=&Multiply_Float(
     A=&Add_Float(
      A=&Lerp_Float(A=1, B=7, Alpha=$Particles.Pulse)
      B=&FloatFromCurve(
       FloatCurve=DataInterface() {propertyValues {Curve {keys [<at 0.5 0 RCIM_Cubic>, {arriveTangent=0.0058, tangentMode=RCTM_User, time=1, value=0}]}}}
       CurveIndex=&ReturnNormalizedExecIndex()
       Scale Curve=&FloatFromCurve(FloatCurve=DataInterface() {propertyValues {Curve {defaultValue=1, keys [{arriveTangent=10.4185, time=0.75, value=0}, <at 0.9969 0>]}}})
      )
     )
     B=&FloatFromCurve(
      FloatCurve=DataInterface() {propertyValues {Curve {keys [<at 0 0.7 RCIM_Cubic>,<at 1 0.5 RCIM_Cubic>]}}}
      CurveIndex=$System.TimeForDis
      Scale Curve=$User.BeamOn
     )
    ))
   &ScaleColor(
    Scale Alpha=&FloatFromCurve(
     FloatCurve=DataInterface() {propertyValues {Curve {
        defaultValue=1
        keys [
         {interpMode=RCIM_Cubic, tangentMode=RCTM_User, time=0, value=0}
         {interpMode=RCIM_Cubic, tangentMode=RCTM_User, time=0.7291, value=1}
         {tangentMode=RCTM_User, time=1, value=0}
        ]
       }}}
     CurveIndex=$System.TimeForDis
     Scale Curve=$User.BeamOn
    )
    Scale RGB=&Multiply_VectorByFloat(
     Vector=<1 1 1>
     Float=&FloatFromCurve(
      FloatCurve=DataInterface() {propertyValues {Curve {
         defaultValue=1
         keys [
          {interpMode=RCIM_Cubic, tangentMode=RCTM_User, time=0, value=0}
          {interpMode=RCIM_Cubic, tangentMode=RCTM_User, time=0.5, value=1}
          {tangentMode=RCTM_User, time=1, value=0}
         ]
        }}}
      CurveIndex=$System.TimeForDis
      Scale Curve=$User.BeamOn
     )
    )
    ScaleA=true
    ScaleRGB=true
   )
  ]
  rendererClasses [&NiagaraSpriteRendererProperties, &NiagaraRibbonRendererProperties]
  renderers [
   {rendererClass=&NiagaraSpriteRendererProperties}
   {rendererClass=&NiagaraRibbonRendererProperties, rendererIndex=1}
  ]
 }
 GiantGlow {
  bEnabled=false
  emitterUpdateScript [
   &EmitterState(Inactive Response=::NewEnumerator1, Life Cycle Mode=::NewEnumerator1, Loop Behavior=::NewEnumerator1, Loop Duration=3, UseLoopDelay=false)
   &SpawnBurst_Instantaneous(Use Spawn Probability=false, UseLoopCountLimit=false)
  ]
  particleSpawnScript [
   &InitializeParticle(
    Color=<rgba 10 10 10 1>
    Lifetime=1.5
    Mass Mode=::NewEnumerator1
    Mesh Scale Mode=::NewEnumerator3
    Mesh Uniform Scale=10000
    UsePositionOffset=false
    Write Material Random=true
   ) {module=InitializeParticle001}
   &DynamicMaterialParameters(
    Index 0 Param 1=1
    Index 0 Param 2=1
    Index 0 Param 3=1
    Index 0 Param 4=1
    Module.Index0Param0Write=true
    Module.Index0Param1Write=false
    Module.Index0Param2Write=false
    Module.Index0Param3Write=false
   )
  ]
  particleUpdateScript [
   &ParticleState()
   &ScaleColor(
    Scale Alpha=&FloatFromCurve(
     FloatCurve=DataInterface() {propertyValues {Curve {keys [{interpMode=RCIM_Cubic, tangentMode=RCTM_User, time=0, value=0}, {interpMode=RCIM_Cubic, tangentMode=RCTM_User, time=0.2, value=1}, <at 1 0>]}}}
     Scale Curve=0.1
    )
    ScaleA=true
    ScaleRGB=true
   ) {module=ScaleColor001}
  ]
  rendererClasses [&NiagaraMeshRendererProperties]
  renderers [{rendererClass=&NiagaraMeshRendererProperties, rendererIndex=0}]
 }
 PostRadialBlur {
  emitterUpdateScript [
   &EmitterState(Inactive Response=::NewEnumerator1, Life Cycle Mode=::NewEnumerator1, Loop Behavior=::NewEnumerator1, Loop Duration=4, UseLoopDelay=false)
   &SpawnBurst_Instantaneous(Use Spawn Probability=false, UseLoopCountLimit=true)
  ]
  particleSpawnScript [
   &InitializeParticle(
    Lifetime=&Add_Float(A=$User.Duration, B=0)
    Mesh Scale Mode=::NewEnumerator3
    Sprite Size Mode=::NewEnumerator3
    Uniform Sprite Size=&Multiply_Float(A=2456.8269, B=2)
    UsePositionOffset=false
    Write Material Random=true
   )
   &CameraOffset(Camera Offset Amount=1500)
  ]
  particleUpdateScript [
   &ParticleState()
   &ScaleSpriteSize(Uniform Curve Sprite Scale=DataInterface() {propertyValues {Curve {
       defaultValue=0
       keys [
        {tangentMode=RCTM_User, time=0, value=1}
        {arriveTangent=-0.0405, interpMode=RCIM_Cubic, leaveTangent=-0.0405, tangentMode=RCTM_User, time=0.7009, value=1}
        {arriveTangent=-8.0089, interpMode=RCIM_Cubic, leaveTangent=-8.0089, tangentMode=RCTM_User, time=1, value=0}
       ]
      }}})
   &Timeline(Play Rate=2, Timeline End=1.0000000200408773e+20)
   /SetVariables(Particles.float:NiagaraFloat=&FloatFromCurve(
     FloatCurve=DataInterface() {propertyValues {Curve {keys [
         {arriveTangent=18.957, interpMode=RCIM_Cubic, leaveTangent=18.957, tangentMode=RCTM_User, time=0, value=0}
         {arriveTangent=-0.1384, interpMode=RCIM_Cubic, leaveTangent=-0.1384, tangentMode=RCTM_User, time=0.05, value=0.9847}
         {interpMode=RCIM_Cubic, leaveTangent=-3.562, time=1, value=1}
        ]}}}
     Scale Curve=&Waveform(Global Amplitude Scale=1.5, "Waveform [1]"=::NewEnumerator12, "[1] Frequency"=20, "[1] Amplitude Min/Max"=<0.5 1>)
    )) {isSetParameters=true}
   &DynamicMaterialParameters(
    Index 0 Param 1=$Particles.float
    Index 0 Param 2=$Particles.float
    Index 0 Param 3=0.8
    Index 1 Param 1=0.2
    Index 1 Param 2=500
    Index 1 Param 3=0.2
    Index 2 Param 1=$Particles.Timeline.Time
    Index 2 Param 2=$Particles.Timeline.Time
    Index 2 Param 3=$Particles.Timeline.Time
    Index 2 Param 4=0.5
    Index 3 Param 1=2
    Index 3 Param 2=200
    Module.Index0Param0Write=true
    Module.Index0Param1Write=true
    Module.Index0Param2Write=true
    Module.Index0Param3Write=true
    Module.Index1Param0Write=true
    Module.Index1Param1Write=true
    Module.Index1Param2Write=true
    Module.Index1Param3Write=true
    Module.Index2Param0Write=true
    Module.Index2Param1Write=true
    Module.Index2Param2Write=true
    Module.Index2Param3Write=true
    Module.Index3Param0Write=true
    Module.Index3Param1Write=true
    Module.Index3Param2Write=true
    Module.Index3Param3Write=true
    Write Parameter Index 1=true
    Write Parameter Index 2=true
    Write Parameter Index 3=true
   )
   &ScaleColor(
    Scale RGB=&Multiply_VectorByFloat(
     Vector=&VectorFromFloat(Value=&Waveform("Waveform [1]"=::NewEnumerator12, "[1] Frequency"=1, "[1] Amplitude Min/Max"=<1 1>))
     Float=&FloatFromCurve(
      FloatCurve=DataInterface() {propertyValues {Curve {keys [<at 0 1>, <at 0.2 1>, {arriveTangent=0.5082, time=1, value=1}]}}}
      CurveIndex=$Particles.float
     )
    )
    ScaleA=true
    ScaleRGB=true
   )
   &ScaleColor(
    Scale Alpha=&FloatFromCurve(FloatCurve=DataInterface() {propertyValues {Curve {keys [<at 0.8 1>,<at 1 0>]}}})
    ScaleA=true
    ScaleRGB=true
   ) {enabled=false, module=ScaleColor001}
   &ScaleColor(
    Scale Alpha=&FloatFromCurve(
     FloatCurve=DataInterface() {propertyValues {Curve {keys [{interpMode=RCIM_Cubic, leaveTangent=-3.7201, tangentMode=RCTM_Break, time=0, value=1}, <at 1 0>]}}}
     CurveIndex=$System.TimeForDis
    )
    ScaleA=true
    ScaleRGB=true
   ) {module=ScaleColor002}
   /SetVariables(Particles.SpeedNormalizedB:NiagaraFloat=&Multiply_Float(
     A=&FloatFromCurve(
      FloatCurve=DataInterface() {propertyValues {Curve {keys [<at 0.75 0.95 RCIM_Cubic>]}}}
      Scale Curve=&Waveform(
       "Waveform [1]"=::NewEnumerator12
       "[1] Frequency"=20
       "[1] Amplitude Min/Max"=&Lerp_Vector2(
        A=<0.8 0.8>
        B=<0.75 0.8>
        Alpha=&FloatFromCurve(FloatCurve=DataInterface() {propertyValues {Curve {keys [<at 0 0>,<at 0.2 1>]}}}, CurveIndex=$Engine.System.Age)
       )
      )
     )
     B=&FloatFromCurve(
      FloatCurve=DataInterface() {propertyValues {Curve {keys [{interpMode=RCIM_Cubic, leaveTangent=-0.3914, tangentMode=RCTM_Break, time=0.4, value=1}, <at 2 0>]}}}
      CurveIndex=$System.TimeForDis
      Scale Curve=&Multiply_Float(
       A=1
       B=&FN_LOD_Distance_Remap(Start Distance=4000, End Distance=7000, Minimum Output=1, Maximum Output=0, Clamp Output=true)
      )
     )
    )) {isSetParameters=true}
   /SetVariables(Particles.ComponentsEnabled:NiagaraBool=&SetBoolByFloatComparison(A=$Engine.System.Age, B=&Add_Float(A=$User.Duration, B=0.5), Comparison Type=::NewEnumerator6)) {isSetParameters=true}
  ]
  rendererClasses [&NiagaraSpriteRendererProperties]
  renderers [{rendererClass=&NiagaraSpriteRendererProperties}]
 }
 RainbowMFFlare {
  emitterUpdateScript [
   &EmitterState(Life Cycle Mode=::NewEnumerator1, Loop Behavior=::NewEnumerator1, Loop Delay=0.2, Loop Duration=1, UseLoopDelay=false)
   &SpawnBurst_Instantaneous(Use Spawn Probability=false, UseLoopCountLimit=false)
  ]
  particleSpawnScript [
   &InitializeParticle(
    Color=<rgba 10 2.8572 0.3434 1>
    Position Offset=<0 0 120>
    Sprite Rotation Angle=&Add_Float(A=0, B=0)
    Sprite Rotation Mode=::NewEnumerator0
    Sprite Size=<1000 200>
    Sprite Size Mode=::NewEnumerator4
    UsePositionOffset=false
    Write Material Random=true
   )
   &CameraOffset(Camera Offset Amount=800)
   &AddVelocity(Velocity=<0 0 10>)
  ]
  particleUpdateScript [
   /SetVariables(Particles.Position=&AddVectorToPosition(
     Position=$System.StartPoint
     Vector=&TransformVector(Vector=$System.OrbWPO, Source Space=::World, Destination Space=::Local)
    )) {isSetParameters=true}
   &ParticleState()
   &SpriteRotationRate(Rotation Rate=&FloatFromCurve(FloatCurve=DataInterface() {propertyValues {Curve {keys [<at 0 1>,<at 1 0>]}}}, Scale Curve=164.9016)) {enabled=false}
   &ScaleSpriteSize(
    Uniform Curve Scale=&FN_LOD_Distance_Remap(Start Distance=142.1609, End Distance=50000, Minimum Output=0.5, Maximum Output=31.1587, Clamp Output=true)
    Uniform Curve Sprite Scale=DataInterface() {propertyValues {Curve {keys [{interpMode=RCIM_Cubic, leaveTangent=2.6829, tangentMode=RCTM_Break, time=0, value=0}, {tangentMode=RCTM_User, time=1, value=1}]}}}
   )
   &Drag(Drag=10, Use Linear Drag=true, Use Rotational Drag=true)
   &SolveForcesAndVelocity(Clamp Velocity=false, Limit Acceleration=false)
   &DynamicMaterialParameters(
    Index 0 Param 1=&FloatFromCurve(FloatCurve=DataInterface() {propertyValues {Curve {keys [
         {interpMode=RCIM_Cubic, leaveTangent=-1.7135, tangentMode=RCTM_Break, time=0, value=1.2831}
         {arriveTangent=-5.596, tangentMode=RCTM_User, time=1, value=-1.7018}
        ]}}})
    Module.Index0Float0Write=true
    Module.Index0Float1Write=true
    Module.Index0Float2Write=true
    Module.Index0Float3Write=true
   )
   /SetVariables(Particles.SpriteSize2=&Multiply_Vector2D_ByFloat(Float=1.5, Vector2D=$Particles.SpriteSize)) {isSetParameters=true}
   /SetVariables(Particles.Color2=&Multiply_LinearColorByFloat(Linear Color=$Particles.Color, Float=2)) {isSetParameters=true}
   &ScaleColor(
    Scale Alpha=&FloatFromCurve(FloatCurve=DataInterface() {propertyValues {Curve {keys [{interpMode=RCIM_Cubic, tangentMode=RCTM_User, time=0, value=1}, <at 1 0>]}}})
    ScaleA=true
    ScaleRGB=true
   )
  ]
  rendererClasses [&NiagaraSpriteRendererProperties]
  renderers [{rendererClass=&NiagaraSpriteRendererProperties}]
 }
}
summary {
 systemName=NS_DragonCartNarrative_Orb_DeathLaser
 emitters [
  {
   emitterName=Excess_Energy_Beam
   rendererClasses [&NiagaraSpriteRendererProperties, &NiagaraRibbonRendererProperties]
  }
  {
   emitterName=Beam_Core_and_MF
   rendererClasses [
    &NiagaraSpriteRendererProperties
    &NiagaraLightRendererProperties
    &NiagaraMeshRendererProperties
   ]
  }
  {emitterName=Beam_Lightning, rendererClasses [&NiagaraMeshRendererProperties]}
  {emitterName=PostRadialBlur, rendererClasses [&NiagaraSpriteRendererProperties]}
  {emitterName=Beam_Deactivate_Particles, rendererClasses [&NiagaraSpriteRendererProperties]}
  {emitterName=RainbowMFFlare, rendererClasses [&NiagaraSpriteRendererProperties]}
  {bEnabled=false, emitterName=GiantGlow, rendererClasses [&NiagaraMeshRendererProperties]}
 ]
 userVariables [
  User.BeamOn {type=NiagaraFloat, defaultValue=1, description=""}
  User.bHitReaction {type=NiagaraBool, defaultValue=true}
  User.Color 01 {type=LinearColor, description="", defaultValue {struct=LinearColor, value=<rgba 1 0.3755 0.0636 1>}}
  User.Color 02 {type=LinearColor, description="", defaultValue {struct=LinearColor, value=<rgba 1 0.304 0 1>}}
  User.ColorMult {type=NiagaraFloat, defaultValue=1, description=""}
  User.DisappearTime {type=NiagaraFloat, defaultValue=0.8, description=""}
  User.Duration {type=NiagaraFloat, defaultValue=1, description=""}
  User.EndPoint {type=NiagaraPosition, description="", defaultValue {}}
  User.LightningRadius {type=NiagaraFloat, defaultValue=500, description=""}
  User.PulseRadius {type=NiagaraFloat, defaultValue=2, description=""}
  User.Velocity Control {type=NiagaraFloat, defaultValue=0.5, description=""}
  User.WidthMult {type=NiagaraFloat, defaultValue=0.5, description=""}
 ]
}
