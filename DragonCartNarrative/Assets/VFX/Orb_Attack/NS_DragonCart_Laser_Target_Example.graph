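# Niagara example of a DragonCart targeting laser: a burst-spawned beam that lives for User.TargetingDuration.
# It starts from the User static mesh reader, shoots out along mesh X, then blends onto a wobbling, smoothed point at the system's position.
# TargetingLockCurve tightens tracking, damps wobble, shrinks the mesh and switches color to pulsing red at lock threshold.
# Three sprite renderers and one mesh renderer draw it, using computed midpoint, length and distance-scaled width.
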NiagaraSystem /DragonCartNarrative/Assets/VFX/Orb_Attack/NS_DragonCart_Laser_Target_Example
dependencies {
 usedDataInterfaces [&NiagaraDataInterfaceCurve]
 usedDynamicInputs []
 usedModules [
  &SystemState
  &self:SystemSpawnScript.SystemScriptSource.SystemScriptGraph.NiagaraNodeAssignment_1.SetVariables_#1
 ]
}
emitters {TargetingBeam {
  emitterUpdateScript [
   &EmitterState()
   &SpawnBurst_Instantaneous(Use Spawn Probability=false, UseLoopCountLimit=false)
  ]
  particleSpawnScript [
   &InitializeParticle(
    Color Mode=::NewEnumerator0
    Lifetime=$User.TargetingDuration
    Mesh Scale Mode=::NewEnumerator3
    Mesh Uniform Scale=0.6
    Sprite Size Mode=::NewEnumerator3
    Uniform Sprite Size=1000
    UsePositionOffset=false
    Write Material Random=true
   )
   &DynamicMaterialParameters(
    Index 0 Param 1=2
    Index 0 Param 2=10
    Index 0 Param 3=1
    Index 3 Param 1=10
    Module.Index0Float0Write=true
    Module.Index0Float1Write=true
    Module.Index0Float2Write=true
    Module.Index0Float3Write=true
    Module.Index3Float0Write=true
    Module.Index3Float1Write=true
    Module.Index3Float2Write=true
    Module.Index3Float3Write=true
    Write Parameter Index 3=::NewEnumerator1
   )
   &CameraOffset(Camera Offset Amount=50)
  ]
  particleUpdateScript [
   &ParticleState()
   /Niagara/Modules/Spawn/Location/StaticMeshLocation(
    MaskSpawnGroup=false
    Mesh Sampling Type:ENiagara_StaticSamplingMode=::NewEnumerator1
    SamplePosOffset=true
    Sampled Position Offset:Vector3f=<100 0 0>
    Socket Sampling Mode:ENiagara_StaticSocketMode=::NewEnumerator2
    Static Mesh:NiagaraDataInterfaceStaticMesh=$User.Static Mesh Reader
   )
   /SetVariables(Particles.EndPointStartPosition:NiagaraPosition=&AddVectorToPosition(
     Position=$Particles.StaticMeshLocation.SampledPosition
     Vector=&Multiply_VectorByFloat(
      Vector=&ReturnMeshOrientationAxis(Mesh Orientation Axis=::XAxis, Mesh Orientation Quat=$Particles.MeshOrientation)
      Float=&FloatFromCurve(
       FloatCurve=DataInterface() {propertyValues {Curve {defaultValue=1, keys [{tangentMode=RCTM_User, time=0, value=0}, {tangentMode=RCTM_User, time=0.1, value=1}]}}}
       CurveIndex=$Particles.Age
       Scale Curve=8000
      )
     )
    )) {isSetParameters=true}
   /SetVariables(Particles.Oscillation:Vector3f=&Multiply_VectorByFloat(
     Vector=&MakeVector(
      X=&Waveform("Waveform [1]"=::NewEnumerator0, "[1] Frequency"=0.5, "[1] Amplitude Min/Max"=<-1 1>)
      Y=&Waveform("Waveform [1]"=::NewEnumerator0, "[1] Frequency"=0.6, "[1] Amplitude Min/Max"=<-1 1>)
      Z=&Waveform("Waveform [1]"=::NewEnumerator0, "[1] Frequency"=0.8, "[1] Amplitude Min/Max"=<-1 1>)
     )
     Float=&Lerp_Float(A=100, B=0, Alpha=$System.TargetingLockCurve)
    )) {isSetParameters=true}
   &InterpolateOverTime(
    Rate Of Change:NiagaraFloat=&Lerp_Float(A=3, B=50, Alpha=$System.TargetingLockCurve)
    Target Position Value:NiagaraPosition=&AddVectorToPosition(Position=$Engine.Owner.Position, Vector=$Particles.Oscillation)
    Variable Type=::NewEnumerator6
   )
   /SetVariables(Particles.EndPointPosition:NiagaraPosition=&Lerp_Vector(
     A=Linked() {linkedVariable {name=Particles.EndPointStartPosition, type=NiagaraPosition}}
     B=Linked() {linkedVariable {name=Particles.InterpolateOverTime.InterpolateOverTime_Position.Moving Average, type=NiagaraPosition}}
     Alpha=&FloatFromCurve(
      FloatCurve=DataInterface() {propertyValues {Curve {defaultValue=1, keys [<at 0.1 0 RCIM_Cubic>, {tangentMode=RCTM_User, time=0.5, value=1}]}}}
      CurveIndex=$Particles.Age
     )
    )) {isSetParameters=true}
   /SetVariables(Particles.BeamPosition:NiagaraPosition=&Lerp_Vector(
     A=Linked() {linkedVariable {name=Particles.Position, type=NiagaraPosition}}
     B=Linked() {linkedVariable {name=Particles.EndPointPosition, type=NiagaraPosition}}
     Alpha=0.5
    )) {isSetParameters=true}
   /SetVariables(Particles.BeamSpriteSize:Vector2f=&MakeVector2D(
     X=&FN_LOD_Distance_Remap(Start Distance=2000, End Distance=10000, Minimum Output=20, Maximum Output=70, Clamp Output=true)
     Y=&Multiply_Float(A=&DistanceBetweenPositions(A=$Particles.EndPointPosition, B=$Particles.Position), B=1)
    )) {isSetParameters=true}
   /SetVariables(Particles.EndPointSpriteSize:Vector2f=&Vector2DFromFloat(Value=25)) {isSetParameters=true}
   &SpriteFacingAndAlignment(
    Sprite Alignment=&Subtract_Position(A=$Particles.Position, B=$Particles.EndPointPosition)
    Sprite Facing=&ReturnMeshOrientationAxis(Mesh Orientation Axis=::XAxis, Mesh Orientation Quat=$Particles.MeshOrientation)
    WriteAlignment=true
    WriteFacing=true
   )
   &ScaleColor(
    Scale RGB=&MakeCustomVectorFromBool(
     Bool=&SetBoolByFloatComparison(A=$System.TargetingLockCurve, B=1, Comparison Type=::NewEnumerator4)
     True Vector=&Multiply_VectorByFloat(
      Vector=<5 0 0>
      Float=&Waveform("Waveform [1]"=::NewEnumerator0, "[1] Frequency"=7, "[1] Amplitude Min/Max"=<0 1>)
     )
     False Vector=<1 1 1>
    )
    ScaleA=false
    ScaleRGB=true
   )
   &ScaleMeshSize(Scale Factor=&VectorFromFloat(Value=&Lerp_Float(A=1, B=0.1, Alpha=$System.TargetingLockCurve)))
  ]
  rendererClasses [&NiagaraSpriteRendererProperties, &NiagaraMeshRendererProperties]
  renderers [
   {rendererClass=&NiagaraSpriteRendererProperties}
   {rendererClass=&NiagaraSpriteRendererProperties, rendererIndex=1}
   {rendererClass=&NiagaraSpriteRendererProperties, rendererIndex=2}
   {rendererClass=&NiagaraMeshRendererProperties, rendererIndex=3}
  ]
 }}
summary {
 systemName=NS_DragonCart_Laser_Target_Example
 emitters [{emitterName=TargetingBeam, rendererClasses [&NiagaraSpriteRendererProperties, &NiagaraMeshRendererProperties]}]
 userVariables [
  User.Static Mesh Reader {
   type=NiagaraDataInterfaceStaticMesh
   description=""
   defaultValue {
    struct=NiagaraExt_VariableValue_DataInterface
    value {
     dataInterface=&self:NiagaraDataInterfaceStaticMesh_0
     dataInterfaceClass=&NiagaraDataInterfaceStaticMesh
    }
   }
  }
  User.TargetingDuration {type=NiagaraFloat, defaultValue=5, description=""}
 ]
}
